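# Control Rig for the Professor Pup large male body that procedurally drives only the leg helper bones: kneepads and mechanical leg parts.
# Calf local Z rotation is remapped through a curve into added kneepad rotation, while mech calf, ankle and ball bones use aims and translation snaps to follow leg targets.
# Mech thighs copy thigh twist rotation, switching to main thigh rotation when CurrentLOD_float exceeds 1.
# Ball aims have zero weight, and two kneepad SetBoneRotation nodes are unconnected.
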ControlRigBlueprint /Game/Characters/Player/Male/Large/Bodies/M_LRG_Professor_Pup/Meshes/Legs_Only_CtrlRig
graphs [
 RigVMFunctionLibrary {links [], nodes []}
 RigVMModel {nodes [
   n1 = &RigUnit_BeginExecution()
   n2 = &RigVMFunction_AnimEvalRichCurve(
    Curve=(EditorCurveData=(Keys=((),(Time=1,Value=1)),DefaultValue=340282346638528859811704183484516925440,PreInfinityExtrap=RCCE_Constant,PostInfinityExtrap=RCCE_Constant))
    SourceMaximum=0
    SourceMinimum=-120
    TargetMaximum=40
    TargetMinimum=0
    Value=&RigVMFunction_MathQuaternionToEuler(RotationOrder=ZYX, Value=&RigUnit_GetBoneTransform(Bone=calf_r, Space=LocalSpace).Transform.Rotation).Result.Z
   )
   n3 = &RigVMFunction_MathQuaternionFromEuler(Euler=(X=0,Y=0,Z=0), Euler.Z=@n2.Result)
   /Script/ControlRig.RigUnit_AddBoneTransform(
    Bone=kneepad_1_r
    Transform=(Rotation=(X=0,Y=0,Z=0,W=1),Translation=(X=0,Y=0,Z=0),Scale3D=(X=1,Y=1,Z=1))
    Weight=1
    bPropagateToChildren=true
    ExecutePin=@n1.ExecutePin
    Transform.Rotation=@n3.Result
   ) {is=struct RigVMUnitNode}
   n4 = &RigUnit_BeginExecution()
   n5 = &RigVMFunction_AnimEvalRichCurve(
    Curve=(EditorCurveData=(Keys=((),(Time=1,Value=1)),DefaultValue=340282346638528859811704183484516925440,PreInfinityExtrap=RCCE_Constant,PostInfinityExtrap=RCCE_Constant))
    SourceMaximum=0
    SourceMinimum=-120
    TargetMaximum=40
    TargetMinimum=0
    Value=&RigVMFunction_MathQuaternionToEuler(RotationOrder=ZYX, Value=&RigUnit_GetBoneTransform(Bone=calf_l, Space=LocalSpace).Transform.Rotation).Result.Z
   )
   n6 = &RigVMFunction_MathQuaternionFromEuler(Euler=(X=0,Y=0,Z=0), Euler.Z=@n5.Result)
   &RigUnit_AddBoneTransform(
    Bone=kneepad_1_l
    Transform=(Rotation=(X=0,Y=0,Z=0,W=1),Translation=(X=0,Y=0,Z=0),Scale3D=(X=1,Y=1,Z=1))
    Weight=1
    bPropagateToChildren=true
    ExecutePin=@n4.ExecutePin
    Transform.Rotation=@n6.Result
   )
   n7 = &RigUnit_BeginExecution()
   RigUnit_GetBoneTransform = &RigUnit_GetBoneTransform(Bone=mech_calf_up_r)
   /Script/ControlRig.RigUnit_AimBone(
    Bone=mech_calf_1_r
    DebugSettings=(Scale=10,WorldOffset=(Rotation=(X=0,Y=0,Z=0,W=1),Translation=(X=0,Y=0,Z=0),Scale3D=(X=1,Y=1,Z=1)))
    Primary=(Weight=1,Axis=(X=0.0,Y=0.0,Z=-1),Target=(X=1,Y=0,Z=0),Kind=Location)
    Secondary=(Weight=1,Axis=(X=1,Y=0.0,Z=0.0),Target=(X=1,Y=0,Z=0),Kind=Location)
    Weight=1
    bPropagateToChildren=true
    ExecutePin=@n7.ExecutePin
    Primary.Target=@RigUnit_GetBoneTransform.Transform.Translation
    Secondary.Target=&RigUnit_GetBoneTransform(Bone=mech_ankle_r).Transform.Translation
   ) {is=struct RigVMUnitNode}
   n8 = &RigUnit_BeginExecution()
   n9 = /Script/RigVM.RigVMFunction_MathFloatSelectBool(
    IfFalse=0
    IfTrue=1
    Condition=/Script/RigVM.RigVMFunction_MathFloatGreater(B=1, A=RigVMVariableNode() {variable=CurrentLOD_float}.Value) {is=struct RigVMUnitNode, template="Greater::Execute(in A,in B,out Result)"}.Result
   ) {is=struct RigVMUnitNode}
   /Script/ControlRig.RigUnit_SetBoneRotation(
    Bone=mech_thigh_r
    Space=GlobalSpace
    Weight=1
    bPropagateToChildren=true
    Rotation=/Script/RigVM.RigVMFunction_MathQuaternionSlerp(
     T=@n9.Result
     A=&RigUnit_GetBoneTransform(Bone=thigh_twist_01_r).Transform.Rotation
     B=&RigUnit_GetBoneTransform(Bone=thigh_r).Transform.Rotation
    ) {is=struct RigVMUnitNode, template="Interpolate::Execute(in A,in B,in T,out Result)"}.Result
    ExecutePin=@n8.ExecutePin
   ) {is=struct RigVMUnitNode}
   n10 = &RigUnit_BeginExecution()
   &RigUnit_SetBoneRotation(
    Bone=mech_thigh_l
    Space=GlobalSpace
    Weight=1
    bPropagateToChildren=true
    Rotation=&RigVMFunction_MathQuaternionSlerp(
     T=@n9.Result
     A=&RigUnit_GetBoneTransform(Bone=thigh_twist_01_l).Transform.Rotation
     B=&RigUnit_GetBoneTransform(Bone=thigh_l).Transform.Rotation
    ) {template="Interpolate::Execute(in A,in B,in T,out Result)"}.Result
    ExecutePin=@n10.ExecutePin
   )
   n11 = &RigUnit_BeginExecution()
   /Script/ControlRig.RigUnit_SetBoneTranslation(
    Bone=mech_calf_2_r
    Space=GlobalSpace
    Weight=1
    bPropagateToChildren=true
    ExecutePin=@n11.ExecutePin
    Translation=&RigUnit_GetBoneTransform(Bone=mech_ankle_r).Transform.Translation
   ) {is=struct RigVMUnitNode}
   n12 = &RigUnit_BeginExecution()
   &RigUnit_SetBoneTranslation(
    Bone=mech_calf_2_l
    Space=GlobalSpace
    Weight=1
    bPropagateToChildren=true
    ExecutePin=@n12.ExecutePin
    Translation=&RigUnit_GetBoneTransform(Bone=mech_ankle_l).Transform.Translation
   )
   n13 = &RigUnit_BeginExecution()
   &RigUnit_AimBone(
    Bone=mech_ankle_3_l
    DebugSettings=(Scale=10,WorldOffset=(Rotation=(X=0,Y=0,Z=0,W=1),Translation=(X=0,Y=0,Z=0),Scale3D=(X=1,Y=1,Z=1)))
    Primary=(Weight=1,Axis=(X=0.0,Y=0.0,Z=1.0),Target=(X=1,Y=0,Z=0),Kind=Location)
    Secondary=(Weight=1,Axis=(X=1.0,Y=0.0,Z=0.0),Target=(X=1,Y=0,Z=0),Kind=Location)
    Weight=1
    bPropagateToChildren=true
    ExecutePin=@n13.ExecutePin
    Secondary.Target=&RigUnit_GetBoneTransform(Bone=mech_ball_l).Transform.Translation
    Primary.Target=&RigUnit_GetBoneTransform(Bone=mech_ankle_3_up_l).Transform.Translation
   )
   n14 = &RigUnit_BeginExecution()
   &RigUnit_AimBone(
    Bone=mech_calf_1_l
    DebugSettings=(Scale=10,WorldOffset=(Rotation=(X=0,Y=0,Z=0,W=1),Translation=(X=0,Y=0,Z=0),Scale3D=(X=1,Y=1,Z=1)))
    Primary=(Weight=1,Axis=(X=0.0,Y=0.0,Z=1.0),Target=(X=1,Y=0,Z=0),Kind=Location)
    Secondary=(Weight=1,Axis=(X=1.0,Y=0.0,Z=0.0),Target=(X=1,Y=0,Z=0),Kind=Location)
    Weight=1
    bPropagateToChildren=true
    ExecutePin=@n14.ExecutePin
    Primary.Target=&RigUnit_GetBoneTransform(Bone=mech_calf_up_l).Transform.Translation
    Secondary.Target=&RigUnit_GetBoneTransform(Bone=mech_ankle_l).Transform.Translation
   )
   n15 = &RigUnit_BeginExecution()
   &RigUnit_AimBone(
    Bone=mech_ankle_3_r
    DebugSettings=(Scale=10,WorldOffset=(Rotation=(X=0,Y=0,Z=0,W=1),Translation=(X=0,Y=0,Z=0),Scale3D=(X=1,Y=1,Z=1)))
    Primary=(Weight=1,Axis=(X=0.0,Y=0.0,Z=-1.0),Target=(X=1,Y=0,Z=0),Kind=Location)
    Secondary=(Weight=1,Axis=(X=1.0,Y=0.0,Z=0.0),Target=(X=1,Y=0,Z=0),Kind=Location)
    Weight=1
    bPropagateToChildren=true
    ExecutePin=@n15.ExecutePin
    Primary.Target=&RigUnit_GetBoneTransform(Bone=mech_ankle_3_up_r).Transform.Translation
    Secondary.Target=&RigUnit_GetBoneTransform(Bone=mech_ball_r).Transform.Translation
   )
   n16 = &RigUnit_BeginExecution()
   &RigUnit_AimBone(
    Bone=mech_calf_1_r
    DebugSettings=(Scale=10,WorldOffset=(Rotation=(X=0,Y=0,Z=0,W=1),Translation=(X=0,Y=0,Z=0),Scale3D=(X=1,Y=1,Z=1)))
    Primary=(Weight=1,Axis=(X=0.0,Y=0.0,Z=-1.0),Target=(X=1,Y=0,Z=0),Kind=Location)
    Secondary=(Weight=1,Axis=(X=1.0,Y=0.0,Z=0.0),Target=(X=1,Y=0,Z=0),Kind=Location)
    Weight=1
    bPropagateToChildren=true
    ExecutePin=@n16.ExecutePin
    Primary.Target=&RigUnit_GetBoneTransform(Bone=mech_calf_up_r).Transform.Translation
    Secondary.Target=&RigUnit_GetBoneTransform(Bone=mech_ankle_r).Transform.Translation
   )
   n17 = &RigUnit_BeginExecution()
   &RigUnit_SetBoneTranslation(
    Bone=mech_ball_r
    Space=GlobalSpace
    Weight=1
    bPropagateToChildren=true
    ExecutePin=@n17.ExecutePin
    Translation=&RigUnit_GetBoneTransform(Bone=mech_ball_tgt_r).Transform.Translation
   )
   n18 = &RigUnit_BeginExecution()
   &RigUnit_SetBoneTranslation(
    Bone=mech_ankle_4_r
    Space=GlobalSpace
    Weight=1
    bPropagateToChildren=true
    ExecutePin=@n18.ExecutePin
    Translation=&RigUnit_GetBoneTransform(Bone=foot_r).Transform.Translation
   )
   n19 = &RigUnit_BeginExecution()
   &RigUnit_SetBoneTranslation(
    Bone=mech_ball_l
    Space=GlobalSpace
    Weight=1
    bPropagateToChildren=true
    ExecutePin=@n19.ExecutePin
    Translation=&RigUnit_GetBoneTransform(Bone=mech_ball_tgt_l).Transform.Translation
   )
   n20 = &RigUnit_BeginExecution()
   &RigUnit_SetBoneTranslation(
    Bone=mech_ankle_4_l
    Space=GlobalSpace
    Weight=1
    bPropagateToChildren=true
    ExecutePin=@n20.ExecutePin
    Translation=&RigUnit_GetBoneTransform(Bone=foot_l).Transform.Translation
   )
   n21 = &RigUnit_BeginExecution()
   &RigUnit_AimBone(
    Bone=mech_ball_l
    DebugSettings=(Scale=10,WorldOffset=(Rotation=(X=0,Y=0,Z=0,W=1),Translation=(X=0,Y=0,Z=0),Scale3D=(X=1,Y=1,Z=1)))
    Primary=(Weight=1,Axis=(X=1.0,Y=0.0,Z=0.0),Target=(X=1,Y=0,Z=0),Kind=Location)
    Secondary=(Weight=1,Axis=(X=0,Y=0,Z=1),Target=(X=1,Y=0,Z=0),Kind=Location)
    Weight=0
    bPropagateToChildren=true
    ExecutePin=@n21.ExecutePin
    Primary.Target=&RigUnit_GetBoneTransform(Bone=mech_ankle_l).Transform.Translation
   )
   n22 = &RigUnit_BeginExecution()
   &RigUnit_AimBone(
    Bone=mech_ball_r
    DebugSettings=(Scale=10,WorldOffset=(Rotation=(X=0,Y=0,Z=0,W=1),Translation=(X=0,Y=0,Z=0),Scale3D=(X=1,Y=1,Z=1)))
    Primary=(Weight=1,Axis=(X=-1.0,Y=0.0,Z=0.0),Target=(X=1,Y=0,Z=0),Kind=Location)
    Secondary=(Weight=1,Axis=(X=0,Y=0,Z=1),Target=(X=1,Y=0,Z=0),Kind=Location)
    Weight=0
    bPropagateToChildren=true
    ExecutePin=@n22.ExecutePin
    Primary.Target=&RigUnit_GetBoneTransform(Bone=mech_ankle_r).Transform.Translation
   )
   &RigUnit_SetBoneRotation(Bone=kneepad_1_r, ExecutePin=(), Weight=1, bPropagateToChildren=true, Rotation=@n3.Result)
   &RigUnit_SetBoneRotation(Bone=kneepad_1_l, ExecutePin=(), Weight=1, bPropagateToChildren=true, Rotation=@n6.Result)
  ]}
]
hierarchy [
 root {loc=<0 0 0>, rot=<0 0 0 1>}
 pelvis {parent=root, loc=<0 -0.7321 92.4218>, rot=<0.0127 0.707 -0.0127 -0.707>}
 spine_01 {parent=pelvis, loc=<5.6002 -0.5128 0>, rot=<0 0 0.034 -0.9994>}
 spine_02 {parent=spine_01, loc=<8.0753 0.0139 0>, rot=<0 0 0.0238 -0.9997>}
 spine_03 {parent=spine_02, loc=<7.0393 0 0>, rot=<0 0 0.0149 -0.9999>}
 spine_04 {parent=spine_03, loc=<8.1024 0 0>, rot=<0 0 -0.0132 -0.9999>}
 spine_05 {parent=spine_04, loc=<10.0146 0 0>, rot=<0 0 -0.0015 -1>}
 clavicle_l {parent=spine_05, loc=<2.5856 -0.2866 -4.3777>, rot=<0.0465 -0.7289 -0.009 -0.6829>}
 upperarm_l {parent=clavicle_l, loc=<18.2264 0.1342 -0.001>, rot=<0.0253 -0.2693 0.0013 -0.9627>}
 lowerarm_l {parent=upperarm_l, loc=<29.0071 0.1686 -1.239>, rot=<0.0391 -0.0099 -0.2453 -0.9686>}
 hand_l {parent=lowerarm_l, loc=<25.8607 0.222 1.0275>, rot=<-0.7193 -0.0988 0.1193 -0.6773>}
 index_metacarpal_l {parent=hand_l, loc=<1.6535 0.0183 -2.96>, rot=<0.0158 -0.0124 -0.0173 -0.9996>}
 index_01_l {parent=index_metacarpal_l, loc=<8.3008 0.0357 -0.0524>, rot=<-0.0044 0.0485 0.0924 -0.9945>}
 index_02_l {parent=index_01_l, loc=<3.3588 -0.3003 0.0452>, rot=<-0.002 0.0027 0.3225 -0.9466>}
 index_03_l {parent=index_02_l, loc=<2.5578 -0.1231 -0.0252>, rot=<-0.002 0.0067 0.2558 -0.9667>}
 middle_metacarpal_l {parent=hand_l, loc=<2.1856 -0.1725 -0.5984>, rot=<0.006 0.0168 -0.0329 -0.9993>}
 middle_01_l {parent=middle_metacarpal_l, loc=<7.3275 -0.2447 0.0129>, rot=<-0.0208 0.0042 0.0858 -0.9961>}
 middle_02_l {parent=middle_01_l, loc=<4.0123 0.1401 0.1175>, rot=<-0.0218 0.0258 0.3512 -0.9357>}
 middle_03_l {parent=middle_02_l, loc=<2.6077 0.0534 0.1373>, rot=<0.0082 0.0188 0.1708 -0.9851>}
 pinky_metacarpal_l {parent=hand_l, loc=<1.7616 -0.9753 3.3814>, rot=<-0.0397 0.0814 -0.069 -0.9935>}
 pinky_01_l {parent=pinky_metacarpal_l, loc=<7.1709 -0.4831 -0.1094>, rot=<-0.0654 -0.0035 0.2242 -0.9723>}
 pinky_02_l {parent=pinky_01_l, loc=<2.8274 0.0893 -0.1043>, rot=<0.0037 0.0011 0.2854 -0.9584>}
 pinky_03_l {parent=pinky_02_l, loc=<2.2046 0.0307 -0.0078>, rot=<-0.0002 0.0098 0.1445 -0.9895>}
 ring_metacarpal_l {parent=hand_l, loc=<2.2074 -0.3625 1.3638>, rot=<-0.0284 0.0722 -0.0458 -0.9959>}
 ring_01_l {parent=ring_metacarpal_l, loc=<7.0295 -0.6197 -0.1446>, rot=<-0.0489 -0.0252 0.1892 -0.9804>}
 ring_02_l {parent=ring_01_l, loc=<3.6174 0.4647 -0.1027>, rot=<0 0 0.2806 -0.9598>}
 ring_03_l {parent=ring_02_l, loc=<2.4152 0.114 0.0199>, rot=<0 0 0.1189 -0.9929>}
 thumb_01_l {parent=hand_l, loc=<1.3118 -0.3191 -2.459>, rot=<0.6546 -0.4042 0.0664 -0.6354>}
 thumb_02_l {parent=thumb_01_l, loc=<5.9336 -0.0448 0.0401>, rot=<-0.0037 0.0213 0.298 -0.9543>}
 thumb_03_l {parent=thumb_02_l, loc=<2.2223 0.3682 -0.0907>, rot=<0.0118 -0.013 0.0066 -0.9998>}
 weapon_l {parent=hand_l, loc=<-0.1974 -3.4111 0.3811>, rot=<0 0 0 -1>}
 lowerarm_l_tgt {parent=hand_l, loc=<-0.9178 0.1156 0.082>, rot=<-0.725 -0.0997 -0.1067 -0.673>}
 lowerarm_twist_01_l {parent=lowerarm_l, loc=<17.8529 -0.4187 0.4778>, rot=<0.2125 0.0001 -0.0002 -0.9772>}
 lowerarm_twist_02_l {parent=lowerarm_l, loc=<9.6188 -0.1908 0.169>, rot=<0.0857 0 -0.0002 -0.9963>}
 arm_l_avg {parent=lowerarm_l, loc=<0 0 0.0001>, rot=<-0.0197 0.005 0.1236 -0.9921>}
 elbow_l_block {parent=arm_l_avg, loc=<-6.5963 1.547 2.3954>, rot=<0.6788 0 -0.045 -0.733>}
 elbow_l_block_tip {parent=elbow_l_block, loc=<8.9773 0 0>, rot=<0.1543 0.6915 0.6983 -0.1022>}
 lowerarm_l_start_zero {parent=elbow_l_block_tip, loc=<0 0 0>, rot=<0.0301 -0.0004 0.9697 0.2423>}
 lowerarm_l_start {parent=lowerarm_l_start_zero, loc=<0 0 0>, rot=<0 0 0 -1>}
 lowerarm_l_end {parent=lowerarm_l_start, loc=<22.2919 0 0>, rot=<0 0 0 -1>}
 upperarm_l_tgt {parent=elbow_l_block, loc=<0 0 0>, rot=<0.685 -0.2137 -0.2074 -0.6649>}
 upperarm_twist_01_l {parent=upperarm_l, loc=<7.9707 0.2552 -0.5646>, rot=<-0.1446 0.0003 0.0001 -0.9895>}
 upperarm_twist_02_l {parent=upperarm_l, loc=<19.4847 0.1952 -0.5806>, rot=<-0.0966 0.0003 0.0001 -0.9953>}
 upperarm_follow_l {parent=upperarm_l, loc=<-0.1924 0.0558 0.1025>, rot=<0 0 1 0>}
 upperarm_aim_l {parent=upperarm_follow_l, loc=<0 0 0>, rot=<0 -0.0243 -0.002 -0.9997>}
 upperarm_l_start_zero {parent=upperarm_aim_l, loc=<-6.2404 -5.0267 2.0091>, rot=<-0.1479 -0.9869 -0.0154 -0.0618>}
 upperarm_l_start {parent=upperarm_l_start_zero, loc=<0 0 0>, rot=<0 0 0 1>}
 upperarm_l_end {parent=upperarm_l_start, loc=<17.1031 0 0>, rot=<0 0 0 1>}
 shoulderPad_l {parent=upperarm_l_start, loc=<-2.8715 2.8027 -4.4766>, rot=<0.0225 -0.0682 0.9312 0.3575>}
 tricep_1_l {parent=upperarm_aim_l, loc=<-6.171 2.0395 1.8662>, rot=<-0.0562 -0.0114 0.9715 0.2299>}
 tricep_2_l {parent=tricep_1_l, loc=<12.1165 0 0>, rot=<0 0 -0.4684 -0.8835>}
 tricep_3_l {parent=tricep_2_l, loc=<14.88 0 0>, rot=<0 0 0 -1>}
 clavPivot_end_l_tgt {parent=clavicle_l, loc=<14.626 0.1417 1.1115>, rot=<-0.0881 0.6261 0.0421 -0.7736>}
 clavicle_r {parent=spine_05, loc=<2.5856 -0.2865 4.3777>, rot=<0.7289 0.0465 -0.6829 0.009>}
 upperarm_r {parent=clavicle_r, loc=<-18.2264 -0.1342 0.001>, rot=<0.0253 -0.2693 0.0013 -0.9627>}
 lowerarm_r {parent=upperarm_r, loc=<-29.0072 -0.1686 1.239>, rot=<0.0391 -0.0099 -0.2453 -0.9686>}
 hand_r {parent=lowerarm_r, loc=<-25.8608 -0.222 -1.0274>, rot=<-0.7193 -0.0988 0.1193 -0.6773>}
 index_metacarpal_r {parent=hand_r, loc=<-1.6534 -0.0181 2.9601>, rot=<0.0158 -0.0124 -0.0173 -0.9996>}
 index_01_r {parent=index_metacarpal_r, loc=<-8.3008 -0.0357 0.0524>, rot=<-0.0044 0.0485 0.0924 -0.9945>}
 index_02_r {parent=index_01_r, loc=<-3.3589 0.3003 -0.0453>, rot=<-0.002 0.0027 0.3225 -0.9466>}
 index_03_r {parent=index_02_r, loc=<-2.5578 0.123 0.0252>, rot=<-0.002 0.0067 0.2558 -0.9667>}
 middle_metacarpal_r {parent=hand_r, loc=<-2.1856 0.1726 0.5984>, rot=<0.006 0.0168 -0.0329 -0.9993>}
 middle_01_r {parent=middle_metacarpal_r, loc=<-7.3273 0.2447 -0.0129>, rot=<-0.0208 0.0042 0.0858 -0.9961>}
 middle_02_r {parent=middle_01_r, loc=<-4.0124 -0.14 -0.1175>, rot=<-0.0218 0.0258 0.3512 -0.9357>}
 middle_03_r {parent=middle_02_r, loc=<-2.6077 -0.0533 -0.1373>, rot=<0.0082 0.0188 0.1708 -0.9851>}
 pinky_metacarpal_r {parent=hand_r, loc=<-1.7617 0.9753 -3.3814>, rot=<-0.0397 0.0814 -0.069 -0.9935>}
 pinky_01_r {parent=pinky_metacarpal_r, loc=<-7.1707 0.4832 0.1094>, rot=<-0.0654 -0.0035 0.2242 -0.9723>}
 pinky_02_r {parent=pinky_01_r, loc=<-2.8275 -0.0892 0.1043>, rot=<0.0037 0.0011 0.2854 -0.9584>}
 pinky_03_r {parent=pinky_02_r, loc=<-2.2045 -0.031 0.0078>, rot=<-0.0002 0.0098 0.1445 -0.9895>}
 ring_metacarpal_r {parent=hand_r, loc=<-2.2073 0.3627 -1.3638>, rot=<-0.0284 0.0722 -0.0458 -0.9959>}
 ring_01_r {parent=ring_metacarpal_r, loc=<-7.0297 0.6197 0.1446>, rot=<-0.0489 -0.0252 0.1892 -0.9804>}
 ring_02_r {parent=ring_01_r, loc=<-3.6171 -0.4648 0.1027>, rot=<0 0 0.2806 -0.9598>}
 ring_03_r {parent=ring_02_r, loc=<-2.4153 -0.1141 -0.0199>, rot=<0 0 0.1189 -0.9929>}
 thumb_01_r {parent=hand_r, loc=<-1.3119 0.3192 2.459>, rot=<0.6546 -0.4042 0.0664 -0.6354>}
 thumb_02_r {parent=thumb_01_r, loc=<-5.9336 0.0447 -0.0401>, rot=<-0.0037 0.0213 0.298 -0.9543>}
 thumb_03_r {parent=thumb_02_r, loc=<-2.2223 -0.3682 0.0907>, rot=<0.0118 -0.013 0.0066 -0.9998>}
 weapon_r {parent=hand_r, loc=<0.1975 3.4111 -0.3811>, rot=<0 0 0 1>}
 lowerarm_r_tgt {parent=hand_r, loc=<0.9179 -0.1155 -0.082>, rot=<0.0997 -0.725 0.673 -0.1067>}
 lowerarm_twist_01_r {parent=lowerarm_r, loc=<-17.8485 -1.639 0.0788>, rot=<-0.5477 0.0002 0.0002 -0.8367>}
 lowerarm_twist_02_r {parent=lowerarm_r, loc=<-9.6165 -0.886 0.2872>, rot=<-0.3106 0.0001 0.0003 -0.9505>}
 arm_r_avg {parent=lowerarm_r, loc=<0 0 0>, rot=<-0.0197 0.005 0.1236 -0.9921>}
 elbow_r_block {parent=arm_r_avg, loc=<6.5963 -1.547 -2.3954>, rot=<0 0.6788 0.733 -0.045>}
 elbow_r_block_tip {parent=elbow_r_block, loc=<8.9773 0 0>, rot=<-0.1543 -0.6915 0.6983 -0.1022>}
 lowerarm_r_start_zero {parent=elbow_r_block_tip, loc=<0 0 0>, rot=<-0.0301 0.0004 0.9697 0.2423>}
 lowerarm_r_start {parent=lowerarm_r_start_zero, loc=<0 0 0>, rot=<0 0 0 -1>}
 lowerarm_r_end {parent=lowerarm_r_start, loc=<22.2919 0 0>, rot=<0 0 0 -1>}
 upperarm_r_tgt {parent=elbow_r_block, loc=<0 0 0>, rot=<-0.685 0.2137 -0.2074 -0.6649>}
 upperarm_twist_01_r {parent=upperarm_r, loc=<-7.9678 -0.1539 0.6024>, rot=<-0.058 0.0002 0.0001 -0.9983>}
 upperarm_twist_02_r {parent=upperarm_r, loc=<-19.4774 -0.1946 0.5853>, rot=<-0.0938 0.0002 0.0001 -0.9956>}
 upperarm_follow_r {parent=upperarm_r, loc=<0.1924 -0.0558 -0.1024>, rot=<0 0 0 1>}
 upperarm_aim_r {parent=upperarm_follow_r, loc=<0 0 0>, rot=<0 0.0243 -0.002 -0.9997>}
 upperarm_r_start_zero {parent=upperarm_aim_r, loc=<-6.2404 -5.0267 -2.0091>, rot=<-0.1479 -0.9869 0.0154 0.0618>}
 upperarm_r_start {parent=upperarm_r_start_zero, loc=<0 0 0>, rot=<0 0 0 1>}
 upperarm_r_end {parent=upperarm_r_start, loc=<17.1031 0 0>, rot=<0 0 0 1>}
 shoulderPad_r {parent=upperarm_r_start, loc=<-2.8715 2.8027 4.4766>, rot=<-0.0225 0.0682 0.9312 0.3575>}
 tricep_1_r {parent=upperarm_aim_r, loc=<-6.171 2.0395 -1.8662>, rot=<0.0562 0.0114 0.9715 0.2299>}
 tricep_2_r {parent=tricep_1_r, loc=<12.1165 0 0>, rot=<0 0 -0.4684 -0.8835>}
 tricep_3_r {parent=tricep_2_r, loc=<14.88 0 0>, rot=<0 0 0 -1>}
 clavPivot_end_r_tgt {parent=clavicle_r, loc=<-14.626 -0.1417 -1.1114>, rot=<-0.6261 -0.0881 0.7736 0.0421>}
 neck_01 {parent=spine_05, loc=<10.815 1.9196 0>, rot=<0 0 -0.1864 -0.9825>}
 neck_02 {parent=neck_01, loc=<4.7347 0.3045 0>, rot=<0 0 0.0363 -0.9993>}
 head {parent=neck_02, loc=<5.7632 0.5552 0>, rot=<0 0 0.115 -0.9934>}
 thigh_l {parent=pelvis, loc=<-1.8716 -0.0109 -11.9936>, rot=<0.0012 0.0572 0.0155 -0.9982>}
 calf_l {parent=thigh_l, loc=<-41.9178 -0.2246 0.0926>, rot=<0.0006 -0.0161 -0.0391 -0.9991>}
 calf_twist_01_l {parent=calf_l, loc=<-32.9334 -0.0967 0.0397>, rot=<0.0237 0.0095 -0.0035 -0.9997>}
 calf_twist_02_l {parent=calf_l, loc=<-15.9268 -0.0967 0.0397>, rot=<0.0237 0.0095 -0.0035 -0.9997>}
 foot_l {parent=calf_l, loc=<-38.7057 1.2438 0.2453>, rot=<0.0975 -0.0474 0.0317 -0.9936>}
 ball_l {parent=foot_l, loc=<-6.1614 13.1469 0.6807>, rot=<0.0001 -0.0001 -0.7063 -0.7079>}
 mech_toe_tgt_l {parent=ball_l, loc=<-5.3748 1.7572 -1.0876>, rot=<0.0176 0.1004 -0.0104 -0.9947>}
 mech_toe_l {parent=foot_l, loc=<-7.9312 7.7761 -0.406>, rot=<-0.0583 0.0834 -0.71 -0.6968>}
 mech_ball_tgt_l {parent=foot_l, loc=<-1.8225 -0.1922 -1.8427>, rot=<-0.0975 0.0474 -0.0317 -0.9936>}
 mech_ankle_l {parent=calf_l, loc=<-18.9834 -7.6108 -0.8891>, rot=<-0.0058 -0.0023 0.9829 0.1843>}
 mech_ankle_3_l {parent=mech_ankle_l, loc=<0 0 0>, rot=<0 0 0 1>}
 mech_ankle_4_l {parent=mech_ankle_3_l, loc=<21.5773 -1.1175 1.3529>, rot=<0 0 0 1>}
 mech_ankle_3_up_l {parent=mech_ankle_l, loc=<0 0 7.0638>, rot=<-0.0058 -0.0023 0.9829 -0.1843>}
 mech_ball_l {parent=calf_l, loc=<-40.7114 0.8357 -1.3561>, rot=<0 0.0102 0.1843 -0.9828>}
 thigh_twist_01_l {parent=thigh_l, loc=<-21.6303 0 0>, rot=<0.0013 0 0 -1>}
 dyn_thigh_belt_l {parent=thigh_twist_01_l, loc=<-7.9444 3.4653 5.4455>, rot=<0.5174 -0.6591 0.4288 0.3377>}
 kneepad_1_l {parent=thigh_l, loc=<-43.0652 -2.1066 -0.8096>, rot=<0.0234 -0.0232 0.9786 -0.2032>}
 kneepad_2_l {parent=kneepad_1_l, loc=<-1.8251 -5.5839 0>, rot=<0 0 0 1>}
 mech_calf_1_l {parent=thigh_l, loc=<-43.1755 -2.3249 -0.5243>, rot=<0.0098 -0.0023 0.9813 -0.1924>}
 mech_calf_2_l {parent=mech_calf_1_l, loc=<18.4622 0 0>, rot=<0 0 0 1>}
 achilles_l {parent=mech_calf_2_l, loc=<-8.7654 11.4954 -0.1212>, rot=<0 0 1 0>}
 mech_calf_up_l {parent=thigh_l, loc=<-43.085 -2.3284 3.9671>, rot=<0 0 0 1>}
 thigh_r {parent=pelvis, loc=<-1.8714 -0.0109 11.9936>, rot=<0.0572 -0.0012 0.9982 0.0155>}
 calf_r {parent=thigh_r, loc=<41.918 0.2246 -0.0926>, rot=<0.0006 -0.0161 -0.0391 -0.9991>}
 calf_twist_01_r {parent=calf_r, loc=<32.9336 0.0967 -0.0397>, rot=<0.0237 0.0095 -0.0035 -0.9997>}
 calf_twist_02_r {parent=calf_r, loc=<15.9269 0.0967 -0.0397>, rot=<0.0237 0.0095 -0.0035 -0.9997>}
 foot_r {parent=calf_r, loc=<38.7058 -1.2438 -0.2453>, rot=<0.0975 -0.0474 0.0317 -0.9936>}
 ball_r {parent=foot_r, loc=<6.1614 -13.1469 -0.6807>, rot=<0.0001 -0.0001 -0.7063 -0.7079>}
 mech_toe_tgt_r {parent=ball_r, loc=<5.3748 -1.757 1.0876>, rot=<-0.1004 0.0176 0.9947 -0.0104>}
 mech_toe_r {parent=foot_r, loc=<7.931 -7.7761 0.406>, rot=<-0.0834 -0.0583 0.6968 -0.71>}
 mech_ball_tgt_r {parent=foot_r, loc=<1.8223 0.1921 1.8427>, rot=<-0.0474 -0.0975 0.9936 -0.0317>}
 mech_ankle_r {parent=calf_r, loc=<18.9834 7.6108 0.8891>, rot=<-0.0023 0.0058 0.1843 -0.9829>}
 mech_ankle_3_r {parent=mech_ankle_r, loc=<0 0 0>, rot=<0 0 0 1>}
 mech_ankle_4_r {parent=mech_ankle_3_r, loc=<21.5775 -1.1174 -1.3529>, rot=<0 0 0 1>}
 mech_ankle_3_up_r {parent=mech_ankle_r, loc=<0 0 -7.0638>, rot=<0.0058 0.0023 0.9829 -0.1843>}
 mech_ball_r {parent=calf_r, loc=<40.7114 -0.8357 1.3561>, rot=<0 0.0102 0.1843 -0.9828>}
 thigh_twist_01_r {parent=thigh_r, loc=<21.6303 0 0>, rot=<0.0013 0 0 -1>}
 dyn_thigh_belt_r {parent=thigh_twist_01_r, loc=<7.9446 -3.4653 -5.4455>, rot=<0.6591 0.5174 -0.3377 0.4288>}
 kneepad_1_r {parent=thigh_r, loc=<43.0653 2.1066 0.8096>, rot=<-0.0232 -0.0234 -0.2032 -0.9786>}
 kneepad_2_r {parent=kneepad_1_r, loc=<-1.8251 -5.5839 0>, rot=<0 0 0 1>}
 mech_calf_1_r {parent=thigh_r, loc=<43.1757 2.3249 0.5243>, rot=<-0.0023 -0.0098 -0.1924 -0.9813>}
 mech_calf_2_r {parent=mech_calf_1_r, loc=<18.4622 0 0>, rot=<0 0 0 1>}
 achilles_r {parent=mech_calf_2_r, loc=<-8.7654 11.4954 0.1211>, rot=<0 0 1 0>}
 mech_calf_up_r {parent=thigh_r, loc=<43.0852 2.3284 -3.9672>, rot=<0 0 1 0>}
 cyclePivot_l {parent=pelvis, loc=<0.2928 0.7181 -5.7237>, rot=<-0.0127 -0.707 0.0127 -0.707>}
 cyclePivot_f {parent=cyclePivot_l, loc=<-5.7237 6.6096 -0.0031>, rot=<0 0 0 1>}
 cyclePivot_r {parent=cyclePivot_f, loc=<-5.7237 -6.6096 0.0031>, rot=<0 0 0 1>}
 cyclePivot_b {parent=cyclePivot_r, loc=<5.7237 -2.7599 0.0509>, rot=<0 0 0 1>}
 cyclePivot_c {parent=cyclePivot_b, loc=<0 2.7599 -0.0509>, rot=<0.0002 0 0 -1>}
 axle {parent=cyclePivot_c, loc=<0 -0.0232 17.8215>, rot=<-0.0002 0 0 -1>}
 wheel {parent=axle, loc=<0 0.1953 -0.135>, rot=<0 0 0 1>}
 axle_tgt {parent=wheel, loc=<10 0 0>, rot=<0 0 0 1>}
 seat {parent=axle, loc=<0 0 0>, rot=<0 0 0 1>}
 plate {parent=seat, loc=<-0.0001 -22.5733 1.3594>, rot=<0.0055 -0.7069 0.0054 -0.7073>}
 seat_2 {parent=seat, loc=<0 -9.4525 18.3256>, rot=<-0.1521 -0.1521 0.6906 -0.6906>}
 seatBack {parent=seat_2, loc=<9.4896 0 -1.8241>, rot=<0.0836 0.0836 -0.7022 -0.7022>}
 p_pelvis {parent=seat, loc=<0 -16.7344 27.0027>, rot=<0.0127 0.707 -0.0127 -0.707>}
 p_tail_01 {parent=p_pelvis, loc=<-6.0177 -0.6853 0>, rot=<0.0028 -0.0026 0.6803 -0.7329>}
 p_tail_02 {parent=p_tail_01, loc=<8.777 0 0>, rot=<0 0 0 -1>}
 p_tail_03 {parent=p_tail_02, loc=<9.9995 0 0>, rot=<0 0 0 -1>}
 p_tail_04 {parent=p_tail_03, loc=<9.9995 0 0>, rot=<0 0 0 -1>}
 p_thigh_r {parent=p_pelvis, loc=<-2.0133 3.8959 3.225>, rot=<-0.3377 0.0665 -0.534 -0.7722>}
 p_calf_r {parent=p_thigh_r, loc=<10.3 0 0>, rot=<0 0 0.9628 0.2702>}
 p_foot_r {parent=p_calf_r, loc=<10.3 0 0>, rot=<-0.1728 -0.2761 0.9217 -0.2105>}
 p_ball_r {parent=p_foot_r, loc=<6.6825 0 0>, rot=<0.0026 -0.0538 0.0127 -0.9985>}
 p_toe_r {parent=p_ball_r, loc=<2.3245 0 0>, rot=<0 0 0 1>}
 p_thigh_l {parent=p_pelvis, loc=<-2.0133 3.8959 -3.225>, rot=<0.0487 0.3315 0.7716 -0.5407>}
 p_calf_l {parent=p_thigh_l, loc=<-10.3 0 0>, rot=<0 0 0.9628 0.2702>}
 p_foot_l {parent=p_calf_l, loc=<-10.3011 0 0>, rot=<-0.2751 0.1539 -0.2062 -0.9263>}
 p_ball_l {parent=p_foot_l, loc=<6.6825 0 0>, rot=<-0.0026 0.0538 0.0127 -0.9985>}
 p_toe_l {parent=p_ball_l, loc=<2.3245 0 0>, rot=<0 0 0 1>}
 p_spine_01 {parent=p_pelvis, loc=<3.1546 0.4498 0>, rot=<0 0 0.034 -0.9994>}
 p_spine_05 {parent=p_spine_01, loc=<5.2999 3.8244 0>, rot=<0 0 0.024 -0.9997>}
 p_neck_01 {parent=p_spine_05, loc=<4.9991 8.0887 0>, rot=<0 0 -0.4199 -0.9076>}
 p_head {parent=p_neck_01, loc=<6.4271 0 0>, rot=<0 0 0.387 -0.9221>}
 dyn_ear_r {parent=p_head, loc=<5.0398 2.434 6.2541>, rot=<0.6894 0.219 -0.0142 -0.6903>}
 dyn_ear_l {parent=p_head, loc=<5.0396 2.4343 -6.2542>, rot=<0.7293 0.0427 -0.1894 -0.6561>}
 p_clavicle_r {parent=p_spine_05, loc=<1.3748 7.7674 1.1233>, rot=<-0.0002 0.7246 -0.1104 -0.6803>}
 p_upperarm_r {parent=p_clavicle_r, loc=<4.2016 0 0>, rot=<-0.4119 0.4148 0.7987 -0.1425>}
 p_lowerarm_r {parent=p_upperarm_r, loc=<-7.7043 0 0>, rot=<0 0 -0.3988 -0.917>}
 p_hand_r {parent=p_lowerarm_r, loc=<-8.4559 0 0>, rot=<-0.0421 -0.4368 0.8928 0.1016>}
 p_middle_metacarpal_r {parent=p_hand_r, loc=<1.8445 0 0>, rot=<-0.0346 -0.0557 0.0049 -0.9978>}
 p_middle_01_r {parent=p_middle_metacarpal_r, loc=<2.1246 0 0>, rot=<0 0 0 -1>}
 p_clavicle_l {parent=p_spine_05, loc=<1.3748 7.7674 -1.1233>, rot=<0.0002 -0.7246 -0.1104 -0.6803>}
 p_upperarm_l {parent=p_clavicle_l, loc=<4.2016 0 0>, rot=<0.879 0.441 -0.0028 0.1814>}
 p_lowerarm_l {parent=p_upperarm_l, loc=<7.0014 0 3.215>, rot=<0 -0.1938 0 -0.981>}
 p_hand_l {parent=p_lowerarm_l, loc=<8.4559 0 0>, rot=<0.3224 0.1016 0.042 -0.9402>}
 p_middle_metacarpal_l {parent=p_hand_l, loc=<1.8445 0 0>, rot=<0.0346 0.0557 0.0049 -0.9978>}
 p_middle_01_l {parent=p_middle_metacarpal_l, loc=<2.1246 0 0>, rot=<0 0 0 1>}
 p_foot_ik_l {parent=seat, loc=<6.336 -10.8059 20.7454>, rot=<0 0 0 1>}
 p_foot_ik_r {parent=seat, loc=<-6.336 -10.8059 20.7454>, rot=<0 0 0 1>}
 brake {parent=seat, loc=<-10.017 5.7392 4.5747>, rot=<-0.2116 0.0236 -0.0244 -0.9768>}
 axle_2 {parent=axle, loc=<0 0.1953 -0.135>, rot=<0 0 0 1>}
 clavPivot_offset_l {parent=axle, loc=<11.1172 0.1953 -0.135>, rot=<0 0 1 0>}
 clavPivot_l_zero {parent=clavPivot_offset_l, loc=<-4.0137 0.559 2.2213>, rot=<0.6376 0.0692 0.7641 0.0693>}
 clavPivot_l {parent=clavPivot_l_zero, loc=<0 0 0>, rot=<0 0 0 -1>}
 clavPivot_end_l {parent=clavPivot_l, loc=<22.0196 0 0>, rot=<0 0 0 -1>}
 upperarm_attach_l {parent=clavPivot_end_l, loc=<-0.491 -0.2786 -3.5658>, rot=<0.8354 0.116 0.5359 -0.0378>}
 clavPivot_offset_r {parent=axle, loc=<-11.1172 0.1953 -0.135>, rot=<0 0 0 1>}
 clavPivot_r_zero {parent=clavPivot_offset_r, loc=<-4.0137 -0.559 2.2213>, rot=<-0.0693 0.7641 0.0692 -0.6376>}
 clavPivot_r {parent=clavPivot_r_zero, loc=<0 0 0>, rot=<0 0 0 -1>}
 clavPivot_end_r {parent=clavPivot_r, loc=<22.0196 0 0>, rot=<0 0 0 -1>}
 upperarm_attach_r {parent=clavPivot_end_r, loc=<-0.491 -0.2786 3.5658>, rot=<0.8354 0.116 -0.5359 0.0378>}
 fork_1 {parent=cyclePivot_c, loc=<0 -0.0232 17.8215>, rot=<-0.0002 0 0 -1>}
 fork_2 {parent=fork_1, loc=<0 6.6294 -3.517>, rot=<0 0 0 1>}
 fork_3 {parent=fork_2, loc=<0 15.8688 1.4054>, rot=<-0.0806 0.7016 0.0805 -0.7034>}
 fork_twist {parent=fork_3, loc=<22.8371 0 0>, rot=<0 0 0.0018 -1>}
 handles {parent=fork_twist, loc=<7.696 0 0>, rot=<0 0 0 -1>}
 handle_r {parent=handles, loc=<1.7578 -1.7623 10.8172>, rot=<0.763 -0.1313 0.5249 -0.3536>}
 p_hand_ik_r {parent=handle_r, loc=<5.6668 2.7733 3.0808>, rot=<0.1287 0.0503 0.9225 0.3603>}
 handle_l {parent=handles, loc=<1.8138 -1.7593 -10.6467>, rot=<0.7644 -0.1322 -0.5229 0.3531>}
 p_hand_ik_l {parent=handle_l, loc=<5.6668 2.7733 -3.0807>, rot=<0.9225 -0.3603 0.1287 -0.0503>}
 handle_xx_l {parent=handles, loc=<1.8138 -1.7593 -10.6467>, rot=<0.7644 -0.1322 -0.5229 0.3531>}
 forkBall_r_zero {parent=fork_3, loc=<-0.0224 -0.0011 8.4424>, rot=<0.5366 -0.2484 0.6122 -0.525>}
 forkBall_r {parent=forkBall_r_zero, loc=<0 0 0>, rot=<0 0.0215 0.072 -0.9972>}
 forkBall_end_r {parent=forkBall_r, loc=<18.5215 0 0>, rot=<0 0 0 -1>}
 forkBall_l_zero {parent=fork_3, loc=<0.0225 0.0011 -8.4424>, rot=<-0.5383 0.2497 0.6108 -0.5243>}
 forkBall_l {parent=forkBall_l_zero, loc=<0 0 0>, rot=<0 -0.0215 0.072 -0.9972>}
 forkBall_end_l {parent=forkBall_l, loc=<18.5215 0 0>, rot=<0 0 0 -1>}
 forkArm_r {parent=fork_1, loc=<-18.0455 0.1951 -0.1353>, rot=<0.1685 -0.1685 -0.6867 -0.6867>}
 forkBall_tgt_r {parent=forkArm_r, loc=<7.5094 0.1936 0>, rot=<0 0 0 1>}
 forkArm_l {parent=fork_1, loc=<18.0455 0.1951 -0.1353>, rot=<-0.6867 -0.6867 0.1685 -0.1685>}
 forkBall_tgt_l {parent=forkArm_l, loc=<7.5094 0.1936 0>, rot=<0 0 0 1>}
 pelvis_shift {parent=pelvis, loc=<0.6978 0 0>, rot=<0 0 0.0179 -0.9998>}
 mech_thigh_l {parent=pelvis_shift, loc=<-7.7224 0.5738 -10.1928>, rot=<0.0015 0.0572 -0.0024 -0.9984>}
 mech_thigh_r {parent=pelvis_shift, loc=<-7.7224 0.5738 10.1928>, rot=<0.0572 -0.0015 0.9984 -0.0024>}
 cyclePivot_proxy {parent=pelvis, loc=<0.2928 0.7181 0>, rot=<-0.0125 -0.707 0.0125 -0.707>}
]
variables [CurrentLOD_float {type=double, default=0}]
